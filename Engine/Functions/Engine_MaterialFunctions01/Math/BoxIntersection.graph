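# Material function that intersects a ray with an axis-aligned box, returning entry and exit positions, T0, T1 and Box Thickness.
# It uses the slab method: inverse direction gives per-axis plane distances, reduced to the largest near T0 and smallest far T1.
# As wired, both reductions read R twice and B, never G, and nothing tests for a miss.
# Ray origin and direction default to camera position and vector.
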MaterialFunction /Engine/Functions/Engine_MaterialFunctions01/Math/BoxIntersection
graph {
 properties {
  bExposeToLibrary=true
  description=Returns Entry and Exit points for a Ray intersecting an Axis Aligned Bounding Box. If you need it to be non axis aligned then you will have to Inverse Transform the Ray Origin and Ray End into the local space of the box.
 }
 nodes [
  n1 = Multiply(A=FunctionInput(Preview=CameraVectorWS()) {bUsePreviewValueAsDefault=true, inputName=Ray Direction, sortPriority=1})
  n2 = Divide(B=@n1) {desc=InvRayDir}
  n3 = FunctionInput(Preview=CameraPositionWS()) {bUsePreviewValueAsDefault=true, inputName=Ray Origin}
  n4 = Multiply(A=@n2, B=Subtract(A=FunctionInput() {bUsePreviewValueAsDefault=true, inputName=Box Min, sortPriority=2}, B=@n3)) {desc=1st Plane Intersections}
  n5 = Multiply(A=@n2, B=Subtract(A=FunctionInput() {bUsePreviewValueAsDefault=true, inputName=Box Max, sortPriority=3, previewValue=<1 0 0 1>}, B=@n3)) {desc=2nd Plane Intersections}
  n6 = MaterialFunctionCall(Float3=Min(A=@n4, B=@n5) {desc=Closest Plane Intersections}) {materialFunction=&BreakOutFloat3Components}
  n7 = Max(A=Max(A=@n6.R, B=@n6.R), B=@n6.B) {desc=T0}
  n8 = Multiply(A=@n3) {desc=Ray Origin}
  n9 = Add(A=Multiply(A=FunctionOutput(None=@n7) {outputName=T0, sortPriority=3}, B=@n1), B=@n8)
  n10 = MaterialFunctionCall(Float3=Max(A=@n4, B=@n5) {desc=Furthest Plane Intersections}) {materialFunction=&BreakOutFloat3Components}
  n11 = Min(A=Min(A=@n10.R, B=@n10.R), B=@n10.B) {desc=T1}
  n12 = Add(A=Multiply(A=FunctionOutput(None=@n11) {outputName=T1, sortPriority=4}, B=@n1), B=@n8)
  FunctionOutput(None=Distance(A=FunctionOutput(None=@n9) {outputName=Ray Entry Position}, B=FunctionOutput(None=@n12) {outputName=Ray Exit Position, sortPriority=1})) {outputName=Box Thickness, sortPriority=2}
  VectorParameter() {parameterName=P0, defaultValue=<rgba 0 0 1 0>}
 ]
}
