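# Material function in the Math category, exposed to the library, that solves a quadratic equation ax^2 + bx + c = 0 from scalar A, B and C inputs.
# It computes the discriminant b^2 - 4ac, takes its square root, and forms roots T0 (minus) and T1 (plus) over 2a.
# The discriminant is also output, since nothing guards against negative values or a zero A.
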MaterialFunction /Engine/Functions/Engine_MaterialFunctions01/Math/QuadraticFormula
graph {
 properties {bExposeToLibrary=true, description="Solved a quadratic equation of the form ax^2 + bx + c = 0.", libraryCategoriesText [Math]}
 nodes [
  n1 = FunctionInput() {inputName=B, inputType=FunctionInput_Scalar, sortPriority=1}
  n2 = Multiply(A=@n1) {constB=-1, desc=-b}
  n3 = FunctionInput() {inputName=C, inputType=FunctionInput_Scalar, sortPriority=2}
  n4 = FunctionInput() {inputName=A, inputType=FunctionInput_Scalar}
  n5 = Subtract(A=Multiply(A=@n1, B=@n1) {desc="b^2"}, B=Multiply(A=Multiply(A=@n3, B=@n4) {desc=ac}) {constB=4, desc=4ac}) {desc="b^2 - 4ac"}
  n6 = SquareRoot(None=@n5)
  n7 = Multiply(A=@n4) {constB=2, desc=2a}
  FunctionOutput(None=Divide(A=Subtract(A=@n2, B=@n6) {desc=-}, B=@n7) {desc=T0}) {outputName=T0}
  FunctionOutput(None=Divide(A=Add(A=@n2, B=@n6) {desc=+}, B=@n7) {desc=T1}) {outputName=T1, sortPriority=1}
  FunctionOutput(None=@n5) {outputName=Discriminant, sortPriority=2}
 ]
}
