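# Material function that analytically intersects the per-pixel camera ray with an oriented cylinder, outputting entry/exit positions and distances, thickness, opacity mask and entry normal.
# It moves the ray into a basis built from the Z and X axis inputs, applies XY scale, and solves a quadratic against the radius.
# A static switch adds end caps by swapping out-of-range hits for Z Min/Max plane hits.
# Several leftover parameter nodes are unconnected.
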MaterialFunction /Engine/Functions/Engine_MaterialFunctions01/Math/CylinderIntersection
graph {
 properties {
  bExposeToLibrary=true
  description=Returns an analytical intersection for a cylinder of the specified coordinates.
  libraryCategoriesText [Math]
 }
 nodes [
  n1 = CameraPositionWS()
  n2 = Normalize(VectorInput=Subtract(A=CameraPositionWS(), B=WorldPosition().XYZ))
  n3 = FunctionInput() {
   bUsePreviewValueAsDefault=true
   description="The X axis is the axis used to apply non uniform scale along.  Defaults to the local X axis of the mesh the material is applied to."
   inputName=Z Min-Max
   inputType=FunctionInput_Vector2
   sortPriority=6
   previewValue=<0 100 0 1>
  }
  n4 = MaterialFunctionCall(Float2=@n3) {materialFunction=&BreakOutFloat2Components}
  n5 = CameraPositionWS()
  n6 = FunctionInput(Preview=ActorPositionWS()) {bUsePreviewValueAsDefault=true, description=The center of the cylinder. Defaults to Actor Position., inputName=Origin}
  n7 = FunctionInput(Preview=Transform(None=Constant3Vector() {constant=<rgba 0 0 1 0>}) {transformSourceType=TRANSFORMSOURCE_Local}) {
   bUsePreviewValueAsDefault=true
   description=The Z axis is the axis along the length of the cylinder. Defaults to the local Z axis of the mesh the material is applied to.
   inputName=Z Axis
   sortPriority=3
  }
  n8 = Normalize(VectorInput=@n7)
  n9 = FunctionInput(Preview=Transform(None=Constant3Vector() {constant=<rgba 1 0 0 0>}) {transformSourceType=TRANSFORMSOURCE_Local}) {
   bUsePreviewValueAsDefault=true
   description="The X axis is the axis used to apply non uniform scale along.  Defaults to the local X axis of the mesh the material is applied to."
   inputName=X Axis
   sortPriority=4
  }
  n10 = Normalize(VectorInput=@n9)
  n11 = MaterialFunctionCall!(Vector1=@n8, Vector2=@n10) {materialFunction=/Engine/Functions/Engine_MaterialFunctions02/Math/CreateThirdOrthogonalVector}
  n12 = FunctionInput() {
   bUsePreviewValueAsDefault=true
   description="The X axis is the axis used to apply non uniform scale along.  Defaults to the local X axis of the mesh the material is applied to."
   inputName=XY Scale
   inputType=FunctionInput_Vector2
   sortPriority=4
   previewValue=<1 1 0 1>
  }
  n13 = Constant() {r=1}
  n14 = AppendVector(A=@n12, B=@n13)
  n15 = Multiply(
   A=MaterialFunctionCall!(VectorToTransform=Subtract(A=@n5, B=@n6), BasisX=@n11.Vector2, BasisY=@n11.Vector2, BasisZ=@n11.Vector2) {desc=Origin Vec, materialFunction=/Engine/Functions/Engine_MaterialFunctions02/Math/InverseTransformMatrix}.Transformed Vector 3
   B=@n14
  ) {desc=Non Uniform Scale}
  n16 = ComponentMask(None=@n15) {b=true}
  n17 = Subtract(A=CameraPositionWS(), B=WorldPosition().XYZ)
  n18 = Multiply(
   A=@n14
   B=MaterialFunctionCall(VectorToTransform=Normalize(VectorInput=@n17), BasisX=@n11.Vector2, BasisY=@n11.Vector2, BasisZ=@n11.Vector2) {desc=Directional Vec, materialFunction=&InverseTransformMatrix}.Transformed Vector 3
  )
  n19 = ComponentMask(None=@n18) {b=true}
  n20 = Divide(A=Subtract(A=@n4.R, B=@n16), B=@n19) {desc=Bottom Cap T}
  n21 = Divide(A=Subtract(A=@n4.G, B=@n16), B=@n19) {desc=Top Cap T}
  n22 = MaterialFunctionCall(Float2=@n18) {materialFunction=&BreakOutFloat2Components}
  n23 = MaterialFunctionCall(Float2=@n15) {materialFunction=&BreakOutFloat2Components}
  n24 = MaterialFunctionCall(Float2=@n18) {materialFunction=&BreakOutFloat2Components}
  n25 = Multiply(A=@n23.R, B=@n24.R)
  n26 = Multiply(A=@n23.G, B=@n24.G)
  n27 = MaterialFunctionCall(Float2=@n15) {materialFunction=&BreakOutFloat2Components}
  n28 = Multiply(A=@n27.R, B=@n27.R)
  n29 = Multiply(A=@n27.G, B=@n27.G)
  n30 = FunctionInput() {
   bUsePreviewValueAsDefault=true
   description="The X axis is the axis used to apply non uniform scale along.  Defaults to the local X axis of the mesh the material is applied to."
   inputName=Radius
   inputType=FunctionInput_Scalar
   sortPriority=1
   previewValue=<50 0 0 1>
  }
  n31 = MaterialFunctionCall!(
   A=Add(A=Multiply(A=@n22.R, B=@n22.R) {desc="D.x ^2"}, B=Multiply(A=@n22.G, B=@n22.G) {desc="A D.y ^2"}) {desc=A}
   B=Add(A=Multiply(A=@n25) {constB=2}, B=Multiply(A=@n26) {constB=2}) {desc=B}
   C=Subtract(A=Add(A=@n28, B=@n29), B=Multiply(A=@n30, B=@n30)) {desc=C}
  ) {materialFunction=&/QuadraticFormula}
  n32 = Multiply(A=@n31.T1)
  n33 = Multiply(A=@n14, B=@n18)
  n34 = Add(A=@n15, B=Multiply(A=@n32, B=@n33) {desc=d * t0}) {desc=Ray Entry Test}
  n35 = MaterialFunctionCall(Float2=@n3) {materialFunction=&BreakOutFloat2Components}
  n36 = If(A=ComponentMask(None=@n34) {b=true}, B=@n35.G, "A > B"=Constant(), "A < B"=Constant() {r=1}) {desc=Ray Entry Test}
  n37 = LinearInterpolate(A=@n21, B=@n32, Alpha=@n36) {desc=Top Cap}
  n38 = If(A=ComponentMask(None=@n34) {b=true}, B=@n35.R, "A > B"=Constant() {r=1}, "A < B"=Constant()) {desc=Ray Exit Test}
  n39 = FunctionInput(Preview=StaticBool() {value=true}) {
   bUsePreviewValueAsDefault=true
   description=By default the cylinder will have end caps defined by the Z min and Z max but instructions can be saved when calculating infinite cylinders by setting this to true.
   inputName=Capped?
   inputType=FunctionInput_StaticBool
   sortPriority=5
   previewValue=<1 1 0 1>
  }
  n40 = FunctionOutput(None=StaticSwitch(True=LinearInterpolate(A=@n20, B=@n37, Alpha=@n38) {desc=Bottom Cap}, False=@n32, Value=@n39)) {outputName=T0, sortPriority=5}
  FunctionOutput(None=Add(A=@n1, B=Multiply(A=@n2, B=@n40)) {desc="Ray Entry, World"}) {outputName=Ray Entry Position}
  n41 = Multiply(A=@n31.T0)
  n42 = Add(A=@n15, B=Multiply(A=@n41, B=@n33) {desc=d * t1}) {desc=Ray Exit Test}
  n43 = If(A=ComponentMask(None=@n42) {b=true}, B=@n35.G, "A > B"=Constant(), "A < B"=Constant() {r=1})
  n44 = If(A=ComponentMask(None=@n42) {b=true}, B=@n35.R, "A > B"=Constant() {r=1}, "A < B"=Constant())
  n45 = Clamp(None=Multiply(A=Add(A=@n36, B=@n43), B=Add(A=@n38, B=@n44)))
  n46 = If(A=@n31.Discriminant, B=Constant(), "A > B"=Constant() {r=1}, "A < B"=Constant())
  FunctionOutput(None=StaticSwitch(True=Multiply(A=@n45, B=@n46) {desc=Capped Opacity Mask}, False=Multiply(A=@n46) {desc=Uncapped Opacity Mask}, Value=@n39)) {outputName=Opacity Mask, sortPriority=3}
  n47 = LinearInterpolate(A=@n21, B=@n41, Alpha=@n43)
  n48 = FunctionOutput(None=StaticSwitch(True=LinearInterpolate(A=@n20, B=@n47, Alpha=@n44), False=@n41, Value=@n39)) {outputName=T1, sortPriority=6}
  FunctionOutput(None=Add(A=@n1, B=Multiply(A=@n2, B=@n48)) {desc="Ray Exit, World"}) {outputName=Ray Exit Position, sortPriority=1}
  FunctionOutput(None=StaticSwitch(
    True=Subtract(A=@n40, B=@n48) {desc=Capped Cylinder Thickness}
    False=Subtract(A=@n32, B=@n41) {desc=Uncapped Cylinder Thickness}
    Value=@n39
   )) {outputName=Cylinder Thickness, sortPriority=2}
  n49 = Constant3Vector() {constant=<rgba 0 0 1 0>}
  n50 = Multiply(A=Add(A=@n15, B=Multiply(A=@n18, B=@n40)) {desc="Ray Entry, Local"}, B=Constant3Vector() {constant=<rgba 1 1 0 0>})
  n51 = StaticSwitch(
   True=LinearInterpolate(A=Constant3Vector() {constant=<rgba 0 0 -1 0>}, B=LinearInterpolate(A=@n49, B=@n50, Alpha=@n36) {desc=Top Cap Normals}, Alpha=@n38) {desc=Bottom Cap Normals}
   False=@n50
   Value=@n39
  )
  n52 = MaterialFunctionCall(Vector1=@n8, Vector2=@n10) {materialFunction=&CreateThirdOrthogonalVector}
  FunctionOutput(None=Normalize(VectorInput=MaterialFunctionCall(VectorToTransform=@n51, BasisX=@n52.Vector2, BasisY=@n52.Vector2, BasisZ=@n52.Vector2) {materialFunction=&Transform3x3Matrix}.Transformed Vector)) {outputName=Normal, sortPriority=4}
  n53 = Constant() {r=1}
  MaterialFunctionCall(X=@n53, Y=@n53, Z=Constant() {r=1}) {materialFunction=&MakeFloat3}
  MaterialFunctionCall(X=ScalarParameter() {defaultValue=1, parameterName=NU}, Y=@n13, Z=Constant() {r=1}) {materialFunction=&MakeFloat3}
  Normalize(VectorInput=VectorParameter() {parameterName=X, defaultValue=<rgba 1 0 0 0>}.RGB)
  Normalize(VectorInput=VectorParameter() {parameterName=Z, defaultValue=<rgba 0 0 1 0>}.RGB)
  ScalarParameter() {defaultValue=50, parameterName=Radius}
  ScalarParameter() {defaultValue=1, parameterName=NU}
  ScalarParameter() {defaultValue=100, parameterName=Z Max}
  ScalarParameter() {defaultValue=-100, parameterName=Z Min}
  ScalarParameter() {defaultValue=100, parameterName=Z Max}
  ScalarParameter() {defaultValue=-100, parameterName=Z Min}
  ScalarParameter() {defaultValue=100, parameterName=Z Max}
  ScalarParameter() {defaultValue=-100, parameterName=Z Min}
  VectorParameter() {parameterName=Z, defaultValue=<rgba 0 0 1 0>}
  VectorParameter() {parameterName=X, defaultValue=<rgba 1 0 0 0>}
 ]
}
