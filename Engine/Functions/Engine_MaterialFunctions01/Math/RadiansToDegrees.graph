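# Material function in the engine's Math library that converts an angle from radians to degrees.
# A scalar Radians input, which uses its preview value as the default, is multiplied by a constant approximation of 180/pi and returned as the Degrees output.
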MaterialFunction /Engine/Functions/Engine_MaterialFunctions01/Math/RadiansToDegrees
graph {
 properties {bExposeToLibrary=true, description=Converts from Radians to Degrees., libraryCategoriesText [Math]}
 nodes [
  n1 = FunctionInput() {bUsePreviewValueAsDefault=true, inputName=Radians, inputType=FunctionInput_Scalar, previewValue=<45 0 0 1>}
  FunctionOutput(None=Multiply(A=@n1) {constB=57.2958}) {outputName=Degrees}
 ]
}
