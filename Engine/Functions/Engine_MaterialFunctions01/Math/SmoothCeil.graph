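# SmoothCeil is a Math library material function that returns a scalar ceil with a smooth ramp before each integer.
# It remaps the input's fraction by Ramp Width, clamps it from below at one half, and shapes it with a biased, scaled cosine.
# Adding this to Ceil eases each step into the next over half the Ramp Width, with discontinuities above a width of 2.
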MaterialFunction /Engine/Functions/Engine_MaterialFunctions01/Math/SmoothCeil
graph {
 properties {bExposeToLibrary=true, description=Smooth ceil is like ceil but gives a smooth ramp before each integer value., libraryCategoriesText [Math]}
 nodes [
  Constant2Vector()
  n1 = FunctionInput() {description=The input value to create a smooth ceil for., inputType=FunctionInput_Scalar}
  n2 = FunctionInput() {
   bUsePreviewValueAsDefault=true
   description=The with of the smooth ramp before each integer. Values over 2 will start to have discontinuities.
   inputName=Ramp Width
   inputType=FunctionInput_Scalar
   sortPriority=1
   previewValue=<1 0 0 1>
  }
  n3 = Subtract(A=Frac(None=@n1), B=OneMinus(None=@n2))
  n4 = Cosine(None=Max(A=Divide(A=@n3, B=@n2), B=Constant() {r=0.5}) {desc=It is good to make functions look like a lancer})
  FunctionOutput(None=Add(A=Ceil(None=@n1), B=ConstantBiasScale(None=@n4))) {description=The smooth ceil result.}
  Multiply(A=TextureCoordinate(), B=Constant2Vector())
  TextureCoordinate()
 ]
}
